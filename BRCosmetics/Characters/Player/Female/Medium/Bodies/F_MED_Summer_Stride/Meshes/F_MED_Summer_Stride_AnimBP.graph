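# Post-pose corrective AnimBlueprint for the F_MED_Summer_Stride female body cosmetic, adjusting extra bones as limbs move.
# The AnimGraph copies the attached parent mesh's pose, then chains Modify Bone nodes on extra thigh, upperarm and left clavicle bones.
# Each update maps the owning mesh's thigh, upperarm and left clavicle swing angles from ref pose to clamped alphas.
# These alphas blend in fixed corrective transforms, with each thigh alpha driving both front and back bones.
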AnimBlueprint /BRCosmetics/Characters/Player/Female/Medium/Bodies/F_MED_Summer_Stride/Meshes/F_MED_Summer_Stride_AnimBP
target_skeleton=&/F_MED_Summer_Stride_Skeleton
graphs [
 AnimGraph {nodes [
   n1 = Get thigh_l_alpha_back() -> (thigh_l_alpha_back:double)
   n2 = "Transform (Modify) Bone - Bone: extra_back_thigh_l"(
    ComponentPose=Local To Component(LocalPose=Copy Pose From Mesh() {bUseAttachedParent=true})
    Alpha=@n1
    Translation="-8.241,-1.592,-3.821"
    Rotation="-17.105,87.500,-76.826"
    Scale="1,2.193,2.438"
   ) {
    rotationMode=BMM_Replace
    rotationSpace=BCS_ParentBoneSpace
    scaleMode=BMM_Replace
    scaleSpace=BCS_ParentBoneSpace
    translationMode=BMM_Additive
    alphaBoolBlend {blendOption=Linear}
    boneToModify {boneName=extra_back_thigh_l}
    rotation=<rot -17.105 87.5 -76.826>
    scale {y=2.193, z=2.438}
    translation=<-8.241 -1.592 -3.821>
   }
   n3 = "Transform (Modify) Bone - Bone: extra_front_thigh_l"(ComponentPose=@n2, Alpha=@n1, Translation="-1.931,4.010,0.273", Rotation="-0.356,38.352,0.048", Scale="1.588,6.633,1.000") {
    rotationMode=BMM_Additive
    rotationSpace=BCS_ParentBoneSpace
    scaleMode=BMM_Replace
    scaleSpace=BCS_ParentBoneSpace
    translationMode=BMM_Additive
    translationSpace=BCS_ParentBoneSpace
    alphaBoolBlend {blendOption=Linear}
    boneToModify {boneName=extra_front_thigh_l}
    rotation=<rot -0.356 38.352 0.048>
    scale=<1.588 6.633>
    translation=<-1.931 4.01 0.273>
   }
   n4 = Get thigh_r_alpha_back() -> (thigh_r_alpha_back:double)
   n5 = "Transform (Modify) Bone - Bone: extra_back_thigh_r"(ComponentPose=@n3, Alpha=@n4, Translation="0.000,-6.587,-4.418", Rotation="21.142,46.125,21.807", Scale="1.000,1.014,1.011") {
    rotationMode=BMM_Additive
    rotationSpace=BCS_ParentBoneSpace
    scaleMode=BMM_Replace
    scaleSpace=BCS_ParentBoneSpace
    translationMode=BMM_Additive
    alphaBoolBlend {blendOption=Linear}
    boneToModify {boneName=extra_back_thigh_r}
    rotation=<rot 21.142 46.125 21.807>
    scale {y=1.014, z=1.011}
    translation {y=-6.587, z=-4.418}
   }
   n6 = "Transform (Modify) Bone - Bone: extra_front_thigh_r"(ComponentPose=@n5, Alpha=@n4, Translation="-14.528,1.182,8.315", Rotation="5.289,35.549,0.882", Scale="2.131,5.548,0.601") {
    rotationMode=BMM_Replace
    rotationSpace=BCS_ParentBoneSpace
    scaleMode=BMM_Replace
    scaleSpace=BCS_ParentBoneSpace
    translationMode=BMM_Additive
    alphaBoolBlend {blendOption=Linear}
    boneToModify {boneName=extra_front_thigh_r}
    rotation=<rot 5.289 35.549 0.882>
    scale=<2.131 5.548 0.601>
    translation=<-14.528 1.182 8.315>
   }
   n7 = "Transform (Modify) Bone - Bone: extra_upperarm_l"(
    ComponentPose=@n6
    Alpha=Get upperarm_l_alpha() -> (upperarm_l_alpha:double)
    Translation="0,0,0"
    Rotation="4.794,-23.681,11.021"
    Scale="2.861,2.059,1.056"
   ) {
    rotationMode=BMM_Replace
    rotationSpace=BCS_ParentBoneSpace
    scaleMode=BMM_Replace
    scaleSpace=BCS_ParentBoneSpace
    alphaBoolBlend {blendOption=Linear}
    boneToModify {boneName=extra_upperarm_l}
    rotation=<rot 4.794 -23.681 11.021>
    scale=<2.861 2.059 1.056>
   }
   n8 = "Transform (Modify) Bone - Bone: extra_upperarm_r"(
    ComponentPose=@n7
    Alpha=Get upperarm_r_alpha() -> (upperarm_r_alpha:double)
    Translation="1.418,-6.518,-2.390"
    Rotation="-32.902,-42.125,-21.902"
    Scale="2.286,1.583,1.337"
   ) {
    rotationMode=BMM_Replace
    rotationSpace=BCS_ParentBoneSpace
    scaleMode=BMM_Replace
    scaleSpace=BCS_ParentBoneSpace
    translationMode=BMM_Additive
    translationSpace=BCS_ParentBoneSpace
    alphaBoolBlend {blendOption=Linear}
    boneToModify {boneName=extra_upperarm_r}
    rotation=<rot -32.902 -42.125 -21.902>
    scale=<2.286 1.583 1.337>
    translation=<1.418 -6.518 -2.39>
   }
   n9 = "Transform (Modify) Bone - Bone: extra_clavicle_l"(ComponentPose=@n8, Alpha=Get clavicle_l_alpha() -> (clavicle_l_alpha:double), Translation="1.27,0,1.803", Rotation="0,0,0", Scale="1,1,1") {
    rotationMode=BMM_Replace
    rotationSpace=BCS_ParentBoneSpace
    scaleMode=BMM_Replace
    scaleSpace=BCS_ParentBoneSpace
    translationMode=BMM_Additive
    translationSpace=BCS_ParentBoneSpace
    alphaBoolBlend {blendOption=Linear}
    boneToModify {boneName=extra_clavicle_l}
    translation {x=1.27, z=1.803}
   }
   Output Pose(Result=Component To Local(ComponentPose=@n9)) {layerGroup=None}
  ]}
 EventGraph {nodes [
   Event BlueprintUpdateAnimation()
   n1 = GetOwnerSkeletalMeshComponent()
   > Set thigh_l_alpha_back(thigh_l_alpha_back:double=MapRangeClamped(Value=GetTwistAndSwingAngleOfDeltaRotationFromRefPose(self=@n1, BoneName=thigh_l).OutSwingAngle, InRangeA=26, InRangeB=100)) -> (Output_Get:double)
   > Set thigh_r_alpha_back(thigh_r_alpha_back:double=MapRangeClamped(Value=GetTwistAndSwingAngleOfDeltaRotationFromRefPose(self=@n1, BoneName=thigh_r).OutSwingAngle, InRangeA=125, InRangeB=60)) -> (Output_Get:double)
   > Set upperarm_l_alpha(upperarm_l_alpha:double=MapRangeClamped(Value=GetTwistAndSwingAngleOfDeltaRotationFromRefPose(self=@n1, BoneName=upperarm_l).OutSwingAngle, InRangeA=56, InRangeB=111)) -> (Output_Get:double)
   > Set upperarm_r_alpha(upperarm_r_alpha:double=MapRangeClamped(Value=GetTwistAndSwingAngleOfDeltaRotationFromRefPose(self=@n1, BoneName=upperarm_r).OutSwingAngle, InRangeA=117, InRangeB=137)) -> (Output_Get:double)
   > Set clavicle_l_alpha(clavicle_l_alpha:double=MapRangeClamped(Value=GetTwistAndSwingAngleOfDeltaRotationFromRefPose(self=@n1, BoneName=clavicle_l).OutSwingAngle, InRangeA=96, InRangeB=46)) -> (Output_Get:double)
  ]}
]
variables [
 clavicle_l_alpha {type=number}
 thigh_l_alpha_back {type=number}
 thigh_r_alpha_back {type=number}
 upperarm_l_alpha {type=number}
 upperarm_r_alpha {type=number}
]
