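# Physics asset for the F_MED_Summer_Stride female face accessory, simulating hair and earrings in a variant named for use with backpacks.
# Left, middle and right hair chains, two shorter tail chains and both earrings hang from the head through limited constraints.
# Main hair chains lock twist and generally allow more swing toward the tips.
# Head, neck and upper-spine shapes, including an angled spine_04 box, supply body volumes for collision.
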PhysicsAsset /BRCosmetics/Characters/Player/Female/Medium/Bodies/F_MED_Summer_Stride/Meshes/Parts/F_MED_Summer_Stride_FaceAcc_Physics_for_backpacks
bodies {
 dyn_l_earing [{shapeType=Box, extentX=1.5584, extentY=1.077, extentZ=0.3, shapeName=dyn_l_earing_box, center=<-0.0054 0 0.2015>, rotation=<rot -4.4215 0 0>}]
 dyn_l_hair_02 [{
   shapeType=Capsule
   length=4.4749
   radius=1.452
   shapeName=dyn_l_hair_02_capsule
   center=<3.2654 -0.0075 -0.15>
   rotation=<rot 29.7135 91.8991 -91.8262>
  }]
 dyn_l_hair_03 [
  {
   shapeType=Capsule
   length=3.7457
   radius=1.4182
   shapeName=dyn_l_hair_03_capsule
   center=<3.0696 0.6299 0.0159>
   rotation=<rot 8.698 101.0887 -84.5849>
  }
  {
   shapeType=Box
   extentX=6.3827
   extentY=2.7455
   extentZ=2.6343
   shapeName=dyn_l_hair_03_box
   center=<2.4377 -2.2601 -0.9093>
   rotation=<rot -1.7842 -5.9257 -16.8307>
  }
 ]
 dyn_l_hair_04 [{
   shapeType=Capsule
   length=2.0908
   radius=1.2319
   shapeName=dyn_l_hair_04_capsule
   center=<2.5724 0.3603 0.1988>
   rotation=<rot 16.4169 99.0824 -85.1956>
  }]
 dyn_l_hair_05 [{
   shapeType=Capsule
   length=2.3395
   radius=1.0394
   shapeName=dyn_l_hair_05_capsule
   center=<2.017 0.8331 0.4649>
   rotation=<rot 9.1871 106.4192 -82.3922>
  }]
 dyn_l_hair_tail_01 [{
   shapeType=Capsule
   length=1.1429
   radius=0.5
   shapeName=dyn_l_hair_tail_01_capsule
   center=<0.7515 -0.497 -0.1042>
   rotation=<rot 2.1999 49.5449 -80.6255>
  }]
 dyn_l_hair_tail_02 [{
   shapeType=Capsule
   length=1.1348
   radius=0.5
   shapeName=dyn_l_hair_tail_02_capsule
   center=<1.6995 -0.8866 -0.0344>
   rotation=<rot 0.0165 103.6435 -90.0534>
  }]
 dyn_l_hair_tail_03 [{
   shapeType=Capsule
   length=0.8853
   radius=0.5
   shapeName=dyn_l_hair_tail_03_capsule
   center=<0.7988 -0.2921 0.0026>
   rotation=<rot -2.1742 105.2473 -84.5936>
  }]
 dyn_m_hair_02 [{
   shapeType=Capsule
   length=4.1306
   radius=1.5261
   shapeName=dyn_m_hair_02_capsule
   center=<3.2643 0.1573 -0.0272>
   rotation=<rot -0.9255 92.0699 -90.6264>
  }]
 dyn_m_hair_03 [{
   shapeType=Capsule
   length=3.79
   radius=1.4926
   shapeName=dyn_m_hair_03_capsule
   center=<3.2296 0.0836 0.0544>
   rotation=<rot 1.0796 89.8986 -91.1225>
  }]
 dyn_m_hair_04 [{
   shapeType=Capsule
   length=4.7465
   radius=1.546
   shapeName=dyn_m_hair_04_capsule
   center=<4.0931 0.0532 -0.0688>
   rotation=<rot 2.3938 94.0058 -90.317>
  }]
 dyn_m_hair_05 [{
   shapeType=Capsule
   length=3.9184
   radius=1.255
   shapeName=dyn_m_hair_05_capsule
   center=<3.2381 -0.6166 0.3064>
   rotation=<rot 2.9524 75.1704 -89.6272>
  }]
 dyn_m_hair_tail_01 [{
   shapeType=Capsule
   length=2.4916
   radius=0.946
   shapeName=dyn_m_hair_tail_01_capsule
   center=<1.7891 0.0133 0.0342>
   rotation=<rot 6.1071 99.3757 -88.4035>
  }]
 dyn_m_hair_tail_02 [{
   shapeType=Capsule
   length=2.3189
   radius=0.8437
   shapeName=dyn_m_hair_tail_02_capsule
   center=<1.8805 -0.6914 -0.845>
   rotation=<rot 7.529 -109.1534 -64.8709>
  }]
 dyn_r_earing [{shapeType=Box, extentX=1.494, extentY=1.3192, extentZ=0.3, shapeName=dyn_r_earing_box, center=<0 0 -0.1512>}]
 dyn_r_hair_02 [{
   shapeType=Capsule
   length=4.1528
   radius=1.5136
   shapeName=dyn_r_hair_02_capsule
   center=<-3.0347 0.3132 -0.132>
   rotation=<rot -1.142 -91.2844 -89.5785>
  }]
 dyn_r_hair_03 [{
   shapeType=Capsule
   length=3.5621
   radius=1.4184
   shapeName=dyn_r_hair_03_capsule
   center=<-3.2179 0.2002 -0.0008>
   rotation=<rot -7.8302 -90.9682 -88.8504>
  }]
 dyn_r_hair_04 [{
   shapeType=Capsule
   length=1.9658
   radius=1.3414
   shapeName=dyn_r_hair_04_capsule
   center=<-2.367 -0.249 0.0577>
   rotation=<rot 1.4182 -89.2126 -89.2849>
  }]
 dyn_r_hair_05 [{
   shapeType=Capsule
   length=2.1906
   radius=1.0797
   shapeName=dyn_r_hair_05_capsule
   center=<-2.3789 -0.6924 0.1509>
   rotation=<rot -5.935 -80.537 -88.9063>
  }]
 head [
  {shapeType=Sphere, radius=7.1057, shapeName=head_sphere, center=<6.0876 -4.3734 0>}
  {shapeType=Capsule, length=3.8198, radius=6.5231, center=<2.0166 4.605 0.2698>, rotation=<rot -23.0538 90.6831 -93.0949>}
  {
   shapeType=Box
   extentX=3.4704
   extentY=4.2194
   extentZ=1.1431
   shapeName=head_box
   center=<3.0722 -0.0768 -6.836>
   rotation=<rot 0 -0.0002 -31.947>
  }
  {
   shapeType=Box
   extentX=4.955
   extentY=3.6671
   extentZ=1.7432
   shapeName=head_box
   center=<2.2839 -0.3256 6.376>
   rotation=<rot 0 -0.0005 35.6012>
  }
 ]
 neck_01 [
  {
   shapeType=Capsule
   length=25.8975
   radius=11.7788
   shapeName=neck_01_capsule
   center=<3.9645 7.0093 1.0844>
   rotation=<rot -0.6376 179.9981 180>
  }
  {shapeType=Capsule, length=30.8278, radius=10.2324, shapeName=neck_01_capsule, center=<13.6815 3.8968 0>}
 ]
 spine_04 [{
   shapeType=Box
   extentX=17.4189
   extentY=3.9232
   extentZ=11.9533
   shapeName=spine_04_box
   center=<25.4073 5.9301 -9.1001>
   rotation=<rot 0 35.847 0>
  }]
 spine_05 [
  {shapeType=Capsule, length=38.9853, radius=11.8728, shapeName=spine_05_capsule, center=<4.5353 4.311 0>}
  {shapeType=Capsule, length=30.3468, radius=11.1983, shapeName=spine_05_capsule, center=<0.891 2.1632 0>}
 ]
}
constraints [
 {bone1Name=dyn_l_hair_02, bone2Name=head, swing1LimitDegrees=10, swing2LimitDegrees=10, twistLimitDegrees=2, twistMotion=Locked}
 {bone1Name=dyn_l_hair_03, bone2Name=dyn_l_hair_02, swing1LimitDegrees=10, swing2LimitDegrees=10, twistLimitDegrees=5, twistMotion=Locked}
 {bone1Name=dyn_l_hair_04, bone2Name=dyn_l_hair_03, swing1LimitDegrees=30, swing2LimitDegrees=10, twistLimitDegrees=10, twistMotion=Locked}
 {bone1Name=dyn_l_hair_05, bone2Name=dyn_l_hair_04, swing1LimitDegrees=30, swing2LimitDegrees=10, twistLimitDegrees=10, twistMotion=Locked}
 {bone1Name=dyn_r_hair_02, bone2Name=head, swing1LimitDegrees=10, swing2LimitDegrees=10, twistLimitDegrees=2, twistMotion=Locked}
 {bone1Name=dyn_r_hair_03, bone2Name=dyn_r_hair_02, swing1LimitDegrees=20, swing2LimitDegrees=10, twistLimitDegrees=5, twistMotion=Locked}
 {bone1Name=dyn_r_hair_04, bone2Name=dyn_r_hair_03, swing1LimitDegrees=20, swing2LimitDegrees=10, twistLimitDegrees=10, twistMotion=Locked}
 {bone1Name=dyn_r_hair_05, bone2Name=dyn_r_hair_04, swing1LimitDegrees=30, swing2LimitDegrees=10, twistLimitDegrees=10, twistMotion=Locked}
 {bone1Name=dyn_m_hair_02, bone2Name=head, swing1LimitDegrees=10, swing2LimitDegrees=10, twistLimitDegrees=2, twistMotion=Locked}
 {bone1Name=dyn_m_hair_03, bone2Name=dyn_m_hair_02, swing1LimitDegrees=10, swing2LimitDegrees=10, twistLimitDegrees=5, twistMotion=Locked}
 {bone1Name=dyn_m_hair_04, bone2Name=dyn_m_hair_03, swing1LimitDegrees=30, swing2LimitDegrees=10, twistLimitDegrees=10, twistMotion=Locked}
 {bone1Name=dyn_m_hair_05, bone2Name=dyn_m_hair_04, swing1LimitDegrees=30, swing2LimitDegrees=10, twistLimitDegrees=10, twistMotion=Locked}
 {bone1Name=dyn_m_hair_tail_01, bone2Name=head, swing1LimitDegrees=10, swing2LimitDegrees=10, twistLimitDegrees=5}
 {bone1Name=dyn_m_hair_tail_02, bone2Name=dyn_m_hair_tail_01, swing1LimitDegrees=15, swing2LimitDegrees=15, twistLimitDegrees=5}
 {bone1Name=dyn_l_hair_tail_01, bone2Name=head, swing1LimitDegrees=5, swing2LimitDegrees=5, twistLimitDegrees=5}
 {bone1Name=dyn_l_hair_tail_02, bone2Name=dyn_l_hair_tail_01, swing1LimitDegrees=10, swing2LimitDegrees=10, twistLimitDegrees=5}
 {bone1Name=dyn_l_hair_tail_03, bone2Name=dyn_l_hair_tail_02, swing1LimitDegrees=15, swing2LimitDegrees=15, twistLimitDegrees=5}
 {bone1Name=dyn_r_earing, bone2Name=head, swing1LimitDegrees=10, swing2LimitDegrees=25, twistLimitDegrees=25}
 {bone1Name=dyn_l_earing, bone2Name=head, swing1LimitDegrees=25, swing2LimitDegrees=10, twistLimitDegrees=25}
]
