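# Animation Blueprint for the Summer Stride female body's face accessory, driving its dynamic hair bones on top of the pose copied from the parent mesh.
# A backpack check blends between two neck-relative RigidBody simulations with separate physics assets; the non-backpack branch adds skydiving bone offsets.
# Layered additive bone corrections then reshape the hair, weighted by head and neck deltas from spine_05 and states like crouching and movement.
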AnimBlueprint /BRCosmetics/Characters/Player/Female/Medium/Bodies/F_MED_Summer_Stride/Meshes/Parts/F_MED_Summer_Stride_FaceAcc_AnimBP
target_skeleton=&/F_MED_Summer_Stride_FaceAcc_Skeleton
graphs [
 AnimGraph {nodes [
   n1 = RigidBody!(
    ComponentPose=Local To Component(LocalPose=Copy Pose From Mesh() {bUseAttachedParent=true})
    Alpha
    OverrideWorldGravity:Vector="0,0,-980"
    ExternalForce
   ) -> (Pose) {
    bClampLinearTranslationLimitToRefPose=true
    bOverrideWorldGravity=true
    overridePhysicsAsset=&/F_MED_Summer_Stride_FaceAcc_Physics
    simulationSpace=BaseBoneSpace
    alphaBoolBlend {blendOption=Linear}
    baseBoneRef {boneName=neck_01}
    componentLinearAccScale=<0.1 0.1 0.1>
    overrideWorldGravity {z=-980}
    simSpaceSettings {
     maxAngularAcceleration=500
     maxAngularVelocity=100
     maxLinearAcceleration=400
     maxLinearVelocity=250
     velocityScaleZ=0
     worldAlpha=1
     externalLinearDragV=<0.1 0.1>
    }
   }
   n2 = Get bIsSkyDiving()
   n3 = "Transform (Modify) Bone - Bone: dyn_r_hair_02"!(ComponentPose=@n1, bAlphaBoolEnabled:bool=@n2, Translation="0,0,0", Rotation="-0.000,0.000,25", Scale="1,1,1") -> (Pose) {
    alphaInputType=Bool
    rotationMode=BMM_Additive
    translationMode=BMM_Additive
    alphaBoolBlend {blendOption=Linear}
    boneToModify {boneName=dyn_r_hair_02}
    rotation {roll=25}
   }
   n4 = "Transform (Modify) Bone - Bone: dyn_m_hair_02"!(ComponentPose=@n3, bAlphaBoolEnabled:bool=@n2, Translation="0,0,0", Rotation="-0.000,0.000,30", Scale="1,1,1") -> (Pose) {
    alphaInputType=Bool
    rotationMode=BMM_Additive
    translationMode=BMM_Additive
    alphaBoolBlend {blendOption=Linear}
    boneToModify {boneName=dyn_m_hair_02}
    rotation {roll=30}
   }
   n5 = "Transform (Modify) Bone - Bone: dyn_l_hair_02"!(ComponentPose=@n4, bAlphaBoolEnabled:bool=@n2, Translation="0,0,0", Rotation="0,0,35", Scale="1,1,1") -> (Pose) {
    alphaInputType=Bool
    rotationMode=BMM_Additive
    translationMode=BMM_Additive
    alphaBoolBlend {blendOption=Linear}
    boneToModify {boneName=dyn_l_hair_02}
    rotation {roll=35}
   }
   n6 = "Transform (Modify) Bone - Bone: dyn_r_hair_04"!(ComponentPose=@n5, bAlphaBoolEnabled:bool=@n2, Translation="0,0,0", Rotation="0.000,0.000,-66.019", Scale="1,1,1") -> (Pose) {
    alphaInputType=Bool
    rotationMode=BMM_Additive
    translationMode=BMM_Additive
    alphaBoolBlend {blendOption=Linear}
    boneToModify {boneName=dyn_r_hair_04}
    rotation {roll=-66.019}
   }
   n7 = "Transform (Modify) Bone - Bone: dyn_m_hair_04"!(ComponentPose=@n6, bAlphaBoolEnabled:bool=@n2, Translation="0,0,0", Rotation="0.000,0.000,-50", Scale="1,1,1") -> (Pose) {
    alphaInputType=Bool
    rotationMode=BMM_Additive
    translationMode=BMM_Additive
    alphaBoolBlend {blendOption=Linear}
    boneToModify {boneName=dyn_m_hair_04}
    rotation {roll=-50}
   }
   n8 = "Transform (Modify) Bone - Bone: dyn_l_hair_04"!(ComponentPose=@n7, bAlphaBoolEnabled:bool=@n2, Translation="0,0,0", Rotation="0.000,0.000,-60.910", Scale="1,1,1") -> (Pose) {
    alphaInputType=Bool
    rotationMode=BMM_Additive
    translationMode=BMM_Additive
    alphaBoolBlend {blendOption=Linear}
    boneToModify {boneName=dyn_l_hair_04}
    rotation {roll=-60.91}
   }
   n9 = Component To Local(ComponentPose=@n8)
   n10 = RigidBody!(ComponentPose=Local To Component(LocalPose=Copy Pose From Mesh() {bUseAttachedParent=true}), Alpha, ExternalForce) -> (Pose) {
    bClampLinearTranslationLimitToRefPose=true
    bOverrideWorldGravity=true
    overridePhysicsAsset=&/F_MED_Summer_Stride_FaceAcc_Physics_for_backpacks
    simulationSpace=BaseBoneSpace
    alphaBoolBlend {blendOption=Linear}
    baseBoneRef {boneName=neck_01}
    componentLinearAccScale=<0.1 0.1 0.1>
    overrideWorldGravity {z=-980}
    simSpaceSettings {
     maxAngularAcceleration=500
     maxAngularVelocity=100
     maxLinearAcceleration=400
     maxLinearVelocity=250
     velocityScaleZ=0
     externalLinearDragV=<0.1 0.1>
    }
   }
   n11 = Component To Local(ComponentPose=@n10)
   n12 = Get backpack_wearing() -> (backpack_wearing:bool)
   n13 = Get head_up() -> (head_up:double)
   n14 = "Transform (Modify) Bone - Bone: dyn_r_hair_02"!(
    ComponentPose=Local To Component(LocalPose=Blend Poses by bool(BlendPose_0=@n11, BlendPose_1=@n9, bActiveValue=@n12) {blendType=Linear})
    Alpha:float=@n13
    Translation="0.000,0.000,2.604"
    Rotation="0,0,0"
    Scale="1,1,1"
   ) -> (Pose) {
    rotationMode=BMM_Additive
    translationMode=BMM_Additive
    alphaBoolBlend {blendOption=Linear}
    boneToModify {boneName=dyn_r_hair_02}
    translation {z=2.604}
   }
   n15 = "Transform (Modify) Bone - Bone: dyn_r_hair_03"(ComponentPose=@n14, Alpha=@n13, Translation="-1.201,-0.364,2.000", Rotation="-7.638,0.000,-0.000", Scale="1,1,1") {
    rotationMode=BMM_Additive
    translationMode=BMM_Additive
    alphaBoolBlend {blendOption=Linear}
    boneToModify {boneName=dyn_r_hair_03}
    rotation {pitch=-7.638}
    translation=<-1.201 -0.364 2>
   }
   n16 = "Transform (Modify) Bone - Bone: dyn_r_hair_04"!(ComponentPose=@n15, Alpha:float=@n13, Translation="0.000,-0.474,1.061", Rotation="-13.230,-5.340,-4.840", Scale="1,1,1") -> (Pose) {
    rotationMode=BMM_Additive
    translationMode=BMM_Additive
    alphaBoolBlend {blendOption=Linear}
    boneToModify {boneName=dyn_r_hair_04}
    rotation=<rot -13.23 -5.34 -4.84>
    translation {y=-0.474, z=1.061}
   }
   n17 = Get neck_abs_noncrouch() -> (neck_abs_noncrouch:double)
   n18 = "Transform (Modify) Bone - Bone: dyn_r_hair_02"(ComponentPose=@n16, Alpha=@n17, Translation="-0.000,-0.782,-0.229", Rotation="3.530,-0.293,18.552", Scale="1,1,1") {
    rotationMode=BMM_Additive
    translationMode=BMM_Additive
    alphaBoolBlend {blendOption=Linear}
    boneToModify {boneName=dyn_r_hair_02}
    rotation=<rot 3.53 -0.293 18.552>
    translation {y=-0.782, z=-0.229}
   }
   n19 = "Transform (Modify) Bone - Bone: dyn_r_hair_03"(ComponentPose=@n18, Alpha=@n17, Translation="0.000,1.533,0.979", Rotation="-19.717,8.067,-17.409", Scale="1,1,1") {
    rotationMode=BMM_Additive
    translationMode=BMM_Additive
    alphaBoolBlend {blendOption=Linear}
    boneToModify {boneName=dyn_r_hair_03}
    rotation=<rot -19.717 8.067 -17.409>
    translation {y=1.533, z=0.979}
   }
   n20 = "Transform (Modify) Bone - Bone: dyn_r_hair_04"(ComponentPose=@n19, Alpha=@n17, Translation="0,0,0", Rotation="-0.000,0.000,3.023", Scale="1,1,1") {
    rotationMode=BMM_Additive
    translationMode=BMM_Additive
    alphaBoolBlend {blendOption=Linear}
    boneToModify {boneName=dyn_r_hair_04}
    rotation {roll=3.023}
   }
   n21 = "Transform (Modify) Bone - Bone: dyn_r_hair_05"(ComponentPose=@n20, Alpha=@n17, Translation="0,0,0", Rotation="0.000,0.000,-35.292", Scale="1,1,1") {
    rotationMode=BMM_Additive
    translationMode=BMM_Additive
    alphaBoolBlend {blendOption=Linear}
    boneToModify {boneName=dyn_r_hair_05}
    rotation {roll=-35.292}
   }
   n22 = Get backpack_head_up() -> (backpack_head_up:double)
   n23 = "Transform (Modify) Bone - Bone: dyn_r_hair_02"(ComponentPose=@n21, Alpha=@n22, Translation="0,0,0", Rotation="-0.000,0.000,5.000", Scale="1,1,1") {
    rotationMode=BMM_Additive
    translationMode=BMM_Additive
    alphaBoolBlend {blendOption=Linear}
    boneToModify {boneName=dyn_r_hair_02}
    rotation {roll=5}
   }
   n24 = "Transform (Modify) Bone - Bone: dyn_m_hair_02"!(ComponentPose=@n23, Alpha:float=@n22, Translation="0,0,0", Rotation="0.000,0.000,-5.000", Scale="1,1,1") -> (Pose) {
    rotationMode=BMM_Additive
    translationMode=BMM_Additive
    alphaBoolBlend {blendOption=Linear}
    boneToModify {boneName=dyn_m_hair_02}
    rotation {roll=-5}
   }
   n25 = "Transform (Modify) Bone - Bone: dyn_l_hair_02"!(ComponentPose=@n24, Alpha:float=@n22, Translation="-0.000,-3.425,-0.000", Rotation="-9.620,-2.599,5.181", Scale="1,1,1") -> (Pose) {
    rotationMode=BMM_Additive
    translationMode=BMM_Additive
    alphaBoolBlend {blendOption=Linear}
    boneToModify {boneName=dyn_l_hair_02}
    rotation=<rot -9.62 -2.599 5.181>
    translation {y=-3.425}
   }
   n26 = "Transform (Modify) Bone - Bone: dyn_l_hair_03"(ComponentPose=@n25, Alpha=@n22, Translation="0,0,0", Rotation="0.000,0.000,-10.000", Scale="1,1,1") {
    rotationMode=BMM_Additive
    translationMode=BMM_Additive
    alphaBoolBlend {blendOption=Linear}
    boneToModify {boneName=dyn_l_hair_03}
    rotation {roll=-10}
   }
   n27 = "Transform (Modify) Bone - Bone: dyn_r_hair_04"(ComponentPose=@n26, Alpha=@n22, Translation="0,0,0", Rotation="0.000,0.000,-30.000", Scale="1,1,1") {
    rotationMode=BMM_Additive
    translationMode=BMM_Additive
    alphaBoolBlend {blendOption=Linear}
    boneToModify {boneName=dyn_r_hair_04}
    rotation {roll=-30}
   }
   n28 = "Transform (Modify) Bone - Bone: dyn_m_hair_04"!(ComponentPose=@n27, Alpha:float=@n22, Translation="0,0,0", Rotation="-0.000,0.000,10.000", Scale="1,1,1") -> (Pose) {
    rotationMode=BMM_Additive
    translationMode=BMM_Additive
    alphaBoolBlend {blendOption=Linear}
    boneToModify {boneName=dyn_m_hair_04}
    rotation {roll=10}
   }
   n29 = "Transform (Modify) Bone - Bone: dyn_r_hair_05"(ComponentPose=@n28, Alpha=@n22, Translation="0,0,0", Rotation="-0.000,0.000,45.000", Scale="1,1,1") {
    rotationMode=BMM_Additive
    translationMode=BMM_Additive
    alphaBoolBlend {blendOption=Linear}
    boneToModify {boneName=dyn_r_hair_05}
    rotation {roll=45}
   }
   n30 = Get abs_backpack_head_up() -> (abs_backpack_head_up:double)
   n31 = "Transform (Modify) Bone - Bone: dyn_r_hair_02"(ComponentPose=@n29, Alpha=@n30, Translation="1.695,0.000,0.744", Rotation="-19.684,-1.383,9.373", Scale="1,1,1") {
    rotationMode=BMM_Additive
    translationMode=BMM_Additive
    alphaBoolBlend {blendOption=Linear}
    boneToModify {boneName=dyn_r_hair_02}
    rotation=<rot -19.684 -1.383 9.373>
    translation {x=1.695, z=0.744}
   }
   n32 = "Transform (Modify) Bone - Bone: dyn_m_hair_02"(ComponentPose=@n31, Alpha=@n30, Translation="-0.000,0.759,0.000", Rotation="-9.391,-3.450,20.284", Scale="1,1,1") {
    rotationMode=BMM_Additive
    translationMode=BMM_Additive
    alphaBoolBlend {blendOption=Linear}
    boneToModify {boneName=dyn_m_hair_02}
    rotation=<rot -9.391 -3.45 20.284>
    translation {y=0.759}
   }
   n33 = "Transform (Modify) Bone - Bone: dyn_l_hair_02"(ComponentPose=@n32, Alpha=@n30, Translation="0,0,0", Rotation="-0.000,0.000,5.000", Scale="1,1,1") {
    rotationMode=BMM_Additive
    translationMode=BMM_Additive
    alphaBoolBlend {blendOption=Linear}
    boneToModify {boneName=dyn_l_hair_02}
    rotation {roll=5}
   }
   n34 = "Transform (Modify) Bone - Bone: dyn_r_hair_04"(ComponentPose=@n33, Alpha=@n30, Translation="0,0,0", Rotation="-0.000,0.000,25.000", Scale="1,1,1") {
    rotationMode=BMM_Additive
    translationMode=BMM_Additive
    alphaBoolBlend {blendOption=Linear}
    boneToModify {boneName=dyn_r_hair_04}
    rotation {roll=25}
   }
   n35 = "Transform (Modify) Bone - Bone: dyn_m_hair_04"(ComponentPose=@n34, Alpha=@n30, Translation="0,0,0", Rotation="0.000,0.000,-20.000", Scale="1,1,1") {
    rotationMode=BMM_Additive
    translationMode=BMM_Additive
    alphaBoolBlend {blendOption=Linear}
    boneToModify {boneName=dyn_m_hair_04}
    rotation {roll=-20}
   }
   n36 = "Transform (Modify) Bone - Bone: dyn_l_hair_04"!(ComponentPose=@n35, Alpha:float=@n30, Translation="0,0,0", Rotation="-0.000,0.000,15.000", Scale="1,1,1") -> (Pose) {
    rotationMode=BMM_Additive
    translationMode=BMM_Additive
    alphaBoolBlend {blendOption=Linear}
    boneToModify {boneName=dyn_l_hair_04}
    rotation {roll=15}
   }
   n37 = Get backpack_crouch_abs_head_up() -> (backpack_crouch_abs_head_up:double)
   n38 = "Transform (Modify) Bone - Bone: dyn_r_hair_02"(ComponentPose=@n36, Alpha=@n37, Translation="0.000,-0.000,0.614", Rotation="4.924,-0.871,9.963", Scale="1,1,1") {
    rotationMode=BMM_Additive
    translationMode=BMM_Additive
    alphaBoolBlend {blendOption=Linear}
    boneToModify {boneName=dyn_r_hair_02}
    rotation=<rot 4.924 -0.871 9.963>
    translation {z=0.614}
   }
   n39 = "Transform (Modify) Bone - Bone: dyn_r_hair_03"(ComponentPose=@n38, Alpha=@n37, Translation="0.000,-0.000,1.012", Rotation="0.000,0.000,-5.000", Scale="1,1,1") {
    rotationMode=BMM_Additive
    translationMode=BMM_Additive
    alphaBoolBlend {blendOption=Linear}
    boneToModify {boneName=dyn_r_hair_03}
    rotation {roll=-5}
    translation {z=1.012}
   }
   n40 = "Transform (Modify) Bone - Bone: dyn_r_hair_04"(ComponentPose=@n39, Alpha=@n37, Translation="0,0,0", Rotation="0.000,0.000,-60.000", Scale="1,1,1") {
    rotationMode=BMM_Additive
    translationMode=BMM_Additive
    alphaBoolBlend {blendOption=Linear}
    boneToModify {boneName=dyn_r_hair_04}
    rotation {roll=-60}
   }
   n41 = "Transform (Modify) Bone - Bone: dyn_r_hair_05"(ComponentPose=@n40, Alpha=@n37, Translation="0,0,0", Rotation="0.000,0.000,-20.000", Scale="1,1,1") {
    rotationMode=BMM_Additive
    translationMode=BMM_Additive
    alphaBoolBlend {blendOption=Linear}
    boneToModify {boneName=dyn_r_hair_05}
    rotation {roll=-20}
   }
   Output Pose(Result=Component To Local(ComponentPose=@n41)) {layerGroup=None}
  ]}
 EventGraph {nodes [
   n1 = Event BlueprintUpdateAnimation()
   > n2 = Sequence()
   n3 = GetOwnerSkeletalMeshComponent()
   Set alpha(
    execute=@n2.then_1
    alpha:double=GetAlphaFromReferenceTransformDelta(SkeletalMeshComponent=@n3, BoneName=head, BaseName=spine_05, InputRange=(X=0,Y=-50), TwistAxis="1,0,0")
   ) -> (Output_Get:double)
   > Set neck(neck:double=GetAlphaFromReferenceTransformDelta(SkeletalMeshComponent=@n3, BoneName=neck_01, BaseName=spine_05, DeltaComponentType=EulerY, InputRange=(X=0,Y=11), TwistAxis="1,0,0")) -> (Output_Get:double)
   > Set backpack_wearing(backpack_wearing:bool=CheckForBackbling()) -> (Output_Get:bool)
   > n4 = Sequence()
   > n5 = Branch(Condition=Get bIsMoving2D())
   > Set not_move2d_interp(not_move2d_interp:double=FInterpTo(Current=Get not_move2d_interp() -> (not_move2d_interp:double), Target=0, DeltaTime=@n1.DeltaTimeX, InterpSpeed=5)) -> (Output_Get:double)
   Set not_move2d_interp(execute=@n5.else, not_move2d_interp=FInterpTo(Current=Get not_move2d_interp(), Target=1, DeltaTime=@n1.DeltaTimeX, InterpSpeed=5))
   n6 = Get alpha() -> (alpha:double)
   Set head_up(
    execute=@n4.then_1
    head_up:double=SelectFloat(A=float * float(A=@n6, B=0.55) {is=title K2Node_PromotableOperator}, B=@n6, bPickA=Get bIsCrouching())
   ) -> (Output_Get:double)
   > Set neck_abs_noncrouch(neck_abs_noncrouch:double=float * float!(A, B=Get neck() -> (neck:double), C:double) -> (ReturnValue)) -> (Output_Get:double)
   > n7 = Set backpack_head_up(backpack_head_up:double=float * float!(A=Get alpha(), B) -> (ReturnValue)) -> (Output_Get:double)
   > Set abs_backpack_head_up(abs_backpack_head_up:double=float * float!(A=@n7, B, C:double=Get not_move2d_interp()) -> (ReturnValue)) -> (Output_Get:double)
   > Set backpack_crouch_abs_head_up(backpack_crouch_abs_head_up:double=float * float!(A=Get alpha(), B, C=Get not_move2d_interp(), D:double, E:double) -> (ReturnValue)) -> (Output_Get:double)
   n8 = Gravity Override State Function DEPRECATED(
    execute=@n2.then_0
    CharacterPartAnimInstance=Self-Reference()
    GameTimeInSeconds=GetGameTimeInSeconds()
    GravityOverrideParams=Get Gravity Override Params()
   )
   > Set GravityOverride(GravityOverride=@n8)
   NOT Boolean(A=Get bIsCrouching())
   Get bIsTargeting()
   Get bIsCrouching()
   Get bIsTargeting()
   Get backpack_wearing()
   Get bIsTargeting()
   Get backpack_wearing()
  ]}
]
variables [
 abs_backpack_head_up {type=number}
 alpha {type=number}
 backpack_crouch_abs_head_up {type=number}
 backpack_head_up {type=number}
 backpack_wearing {type=boolean}
 Gravity Override Params {type=GravityOverrideParamsStruct}
 GravityOverride {type=Vector}
 head_up {type=number}
 neck {type=number}
 neck_abs_noncrouch {type=number}
 not_move2d_interp {type=number, default=1}
]
